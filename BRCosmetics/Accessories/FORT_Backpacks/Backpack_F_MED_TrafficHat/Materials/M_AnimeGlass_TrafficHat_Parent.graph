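# Parent material for stylized anime glass on the TrafficHat backpack, rendered unlit and additive on skeletal meshes.
# Fresnel-blended tints multiply sky light reflections, with a rounded-Fresnel hard rim whose exponent varies through a local-space sphere mask for smaller parts.
# An optional view-space matcap adds specular, then time-of-day blending dims evening and night emissive before tech-art overrides.
# A GGX highlight branch is left unconnected.
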Material /BRCosmetics/Accessories/FORT_Backpacks/Backpack_F_MED_TrafficHat/Materials/M_AnimeGlass_TrafficHat_Parent
graph {
 properties {
  bUseMaterialAttributes=true
  bUsedWithMeshDeformer=true
  bUsedWithSkeletalMesh=true
  blendMode=BLEND_Additive
  pixelDepthOffsetMode=PDOM_Legacy
  shadingModel=MSM_Unlit
  usedShadingModels=Unlit
 }
 nodes [
  n1 = SkyLightEnvMapSample(Direction=ReflectionVectorWS(), Roughness=Constant())
  n2 = LinearInterpolate(
   A=VectorParameter() {parameterName=GlassColor, defaultValue=<rgba 0.0429 0.0885 0.0839 1>}.RGB
   B=VectorParameter() {parameterName=GlassColorb, defaultValue=<rgba 0.2068 0.3281 0.3085 1>}.RGB
   Alpha=Fresnel(ExponentIn=ScalarParameter() {defaultValue=1, parameterName=GlassinnerEXP}) {baseReflectFraction=0}
  )
  n3 = ScalarParameter() {defaultValue=2, parameterName=GlassEdgeEXP}
  n4 = Multiply(A=@n3, B=ScalarParameter() {defaultValue=0.5, parameterName=GlassEdgeEXPInnerMultiplier}) {constB=0.6}
  n5 = NamedRerouteUsage()
  n6 = Round(None=Fresnel(ExponentIn=LinearInterpolate(A=@n3, B=@n4, Alpha=@n5)) {baseReflectFraction=0})
  n7 = Saturate(None=LinearInterpolate(
    A=Multiply(A=@n1, B=@n2)
    B=VectorParameter() {parameterName=GlassColorEDGE, defaultValue=<rgba 4.8268 8 6.417 1>}.RGB
    Alpha=Saturate(None=@n6)
   ))
  n8 = Constant() {r=1}
  n9 = TransformPosition(None=WorldPosition().XYZ) {transformSourceType=TRANSFORMPOSSOURCE_World, transformType=TRANSFORMPOSSOURCE_View}
  n10 = PixelNormalWS()
  n11 = MaterialFunctionCall(Color=CrossProduct(A=Normalize(VectorInput=@n9), B=Transform(None=@n10) {transformSourceType=TRANSFORMSOURCE_World, transformType=TRANSFORM_View})) {materialFunction=&SplitComponents}
  n12 = Multiply(A=AppendVector(A=@n11.G, B=@n11.G)) {constB=0.5}
  n13 = Power(
   Base=TextureSampleParameter2D(UVs=Add(A=@n12) {constB=0.5}) {parameterName=MatcapTexture, texture=&/T_Specular_Matcap}.G
   Exp=ScalarParameter() {defaultValue=1, parameterName=GlassReflectionExponent}
  )
  n14 = StaticSwitchParameter(True=Saturate(None=Add(A=@n7, B=@n13)), False=@n7) {defaultValue=true, parameterName=USeMatcapSpecular}
  n15 = Multiply(A=@n14, B=ScalarParameter() {defaultValue=0.75, parameterName=ToD_Night_Intensity}) {constB=0.5}
  n16 = SetMaterialAttributes!(
   MaterialAttributes
   Base Color=@n7
   Metallic=@n8
   Specular=@n8
   Roughness=@n8
   Opacity=ScalarParameter() {defaultValue=1, parameterName=GlassOpacity}
   Emissive Color=MaterialFunctionCall(Color_Morning=@n14, Color_Day=@n14, Color_Evening=@n15, Color_Night=@n15) {materialFunction=&Color_Time_of_Day}.Result
   Shading Model=ShadingModel()
  ) {attributeSetTypes [
    #1
    #2
    #3
    #4
    #5
    #6
    #7
   ]}
  MaterialFunctionCall(In=@n16) {materialFunction=&MF_TechArtOverrides_accessories}
  n17 = SphereMask(
   A=LocalPosition().XYZ
   B=VectorParameter() {parameterName=SphereMaskPosition, defaultValue=<rgba 0 0 12.2508 1>}.RGB
   Radius=ScalarParameter() {defaultValue=30, parameterName=SphereMaskRadius}
  ) {attenuationRadius=50, hardnessPercent=90}
  NamedRerouteDeclaration(None=LinearInterpolate(A=@n17, B=OneMinus(None=@n17), Alpha=ScalarParameter() {parameterName=SphereMaskRadiusInvert, sliderMax=1})) {name=SphereMask, desc=Use different radius for smaller parts, nodeColor=<rgba 0.7812 1 0 1>}
  n18 = PixelNormalWS()
  n19 = Add(A=SkyAtmosphereLightDirection(), B=CameraVectorWS())
  SmoothStep(Value=Saturate(None=MaterialFunctionCall!(Normal=@n18, Light Vector=@n19, Roughness) {materialFunction=/Engine/Functions/Engine_MaterialFunctions01/Lighting/GGXSpecular}.Result)) {constMax=0.63, constMin=0.5}
 ]
}
