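# Material function that paints an animated swirling liquid effect for the TrafficHat backpack, outputting color, shaded color and normal.
# Two panned texture samples on cylinder-projected UVs, the first distorting the second, feed a time-driven sine gradient that is cel-banded or smooth depending on UsePBR.
# The gradient drives a three-color blend with a Fresnel tint, and screen-space derivatives perturb the normal.
# Results blend over the inputs, with masks built from an FX channel and a local-space sphere.
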MaterialFunction /BRCosmetics/Accessories/FORT_Backpacks/Backpack_F_MED_TrafficHat/Materials/MaterialFunctions/MF_SwirlingLiquid
graph {nodes [
  n1 = VectorParameter() {parameterName=Color 2, defaultValue=<rgba 0.1857 0 0.8021 1>}
  n2 = MaterialFunctionCall!(In) {materialFunction=&/MF_CylinderProjectedUV}
  n3 = Multiply(A=@n2.Result, B=ComponentMask(None=VectorParameter() {defaultValue=<rgba 1 5 0 1>}.RGB) {g=true, r=true}) {constB=0.25}
  n4 = TextureSample!(UVs=Panner(Coordinate=@n3) {speedX=0.05, speedY=0.15}, Tex, "DDX(UVs)"=@n2.DDX, "DDY(UVs)"=@n2.DDX, Apply View MipBias) {
   mipValueMode=TMVM_Derivative
   samplerType=SAMPLERTYPE_LinearColor
   texture=/BRCosmetics/Accessories/FORT_Backpacks/Backpack_F_MED_TrafficHat/Materials/T_TrafficHat_basecolor
  }
  n5 = Multiply(A=Subtract(A=@n4.G) {constB=0.5}) {constB=0.5}
  n6 = TextureSample(UVs=Panner(Coordinate=Add(A=@n3, B=@n5)) {speedX=0.5, speedY=0.1}, "DDX(UVs)"=@n2.DDX, "DDY(UVs)"=@n2.DDX) {
   mipValueMode=TMVM_Derivative
   samplerType=SAMPLERTYPE_LinearColor
   texture=&T_TrafficHat_basecolor
  }
  n7 = SmoothStep(Min=Multiply(A=@n4.B) {constB=0.5}, Value=@n6.R)
  n8 = Time()
  n9 = PreSkinnedNormal()
  n10 = ConstantBiasScale(None=VertexInterpolator(VS=ComponentMask(None=@n9) {b=true}).PS)
  n11 = Add(A=@n7, B=Add(A=Multiply(A=@n8), B=@n10))
  n12 = Multiply(A=@n7, B=ConstantBiasScale(None=Sine(None=@n11)))
  n13 = ScalarParameter() {defaultValue=1.2, parameterName=Test}
  n14 = NamedRerouteDeclaration(None=Saturate(None=Multiply(A=@n12, B=@n13) {constB=3})) {name=Gradient, nodeColor=<rgba 0.2422 0 1 1>}
  n15 = MaterialFunctionCall!(NdL=@n14, Width=Constant() {r=0.25}, Bands Amount=Constant() {r=2}) {materialFunction=/Game/Global/Materials/Character_Parent_2023/Anime_Parent_2023/MaterialFunctions/MF_SmoothCelShadingWithBands}
  n16 = FunctionInput() {bUsePreviewValueAsDefault=true, inputName=UsePBR, inputType=FunctionInput_Scalar, previewValue=<1 0 0 1>}
  n17 = NamedRerouteDeclaration(None=@n16) {name=UsePBR, nodeColor=<rgba 0.1484 0 1 1>}
  n18 = MaterialFunctionCall(
   A=NamedRerouteUsage()
   B=@n1.RGB
   C=LinearInterpolate(A=@n1.RGB, B=VectorParameter() {parameterName=Color 3, defaultValue=<rgba 0 0.5432 1 1>}.RGB, Alpha=NamedRerouteUsage()) {desc=If not PBR use only 2 colors}
   Alpha=Multiply(A=LinearInterpolate(A=@n15.Divide, B=@n14, Alpha=@n17), B=SmoothStep(Value=@n10) {constMax=0.4})
  ) {materialFunction=&3ColorBlend}
  n19 = ScalarParameter() {defaultValue=1, parameterName=ExponentIn}
  n20 = LinearInterpolate(
   A=@n18.Result
   B=Multiply(A=@n18.Result, B=VectorParameter() {parameterName=Param_1, defaultValue=<rgba 1.4245 0 2 1>}.RGB) {constB=2.5}
   Alpha=SmoothStep(Value=Fresnel(ExponentIn=@n19) {baseReflectFraction=0})
  )
  FunctionOutput(None=LinearInterpolate(A=NamedRerouteUsage(), B=@n20, Alpha=NamedRerouteUsage())) {outputName=Color}
  n21 = NamedRerouteUsage()
  n22 = NamedRerouteUsage()
  n23 = DDX(Value=@n22)
  n24 = DDY(Value=@n22)
  n25 = AppendVector(A=Multiply(A=AppendVector(A=@n23, B=@n24)) {constB=-1}, B=Constant() {r=1})
  FunctionOutput(None=LinearInterpolate(
    A=NamedRerouteUsage()
    B=MaterialFunctionCall(BaseNormal=@n21, AdditionalNormal=@n25) {materialFunction=&BlendAngleCorrectedNormals}.Result
    Alpha=NamedRerouteUsage()
   )) {outputName=OutNormal}
  n26 = ScalarParameter() {defaultValue=0.9, parameterName=Dynamic_ColorShaded_Darkening}
  FunctionOutput(None=LinearInterpolate(A=NamedRerouteUsage(), B=Multiply(A=@n20, B=@n26) {constB=0.9}, Alpha=NamedRerouteUsage())) {outputName=ColoreShaded}
  NamedRerouteDeclaration(None=ChannelMaskParameter(None=FunctionInput() {inputName=FXMask, inputType=FunctionInput_Scalar}) {parameterName=LiquidMask}.R) {name=LiquidMask, nodeColor=<rgba 0 0.9766 1 1>}
  NamedRerouteDeclaration(None=FunctionInput() {inputName=BaseColor}) {name=InBaseColor, nodeColor=<rgba 1 0.7031 0 1>}
  NamedRerouteDeclaration(None=FunctionInput() {inputName=ShadedBaseColor}) {name=InShadedColor, nodeColor=<rgba 1 0.0949 0.7737 1>}
  n27 = VertexInterpolator(VS=Multiply(A=LocalPosition() {localOrigin=InstancePreSkinning}.XYZ, B=Constant3Vector() {constant=<rgba 1 1 1 0>}))
  n28 = VectorParameter() {parameterName=SphereMaskOffset, defaultValue=<rgba 0 -5.625 -6.4533 1>}
  n29 = ScalarParameter() {defaultValue=5.9298, parameterName=Radius}
  NamedRerouteDeclaration(None=Multiply(A=SphereMask(A=@n27.PS, B=@n28.RGB, Radius=@n29) {attenuationRadius=1, hardnessPercent=90}, B=NamedRerouteUsage())) {name=VesselMask, nodeColor=<rgba 1 0 0.375 1>}
  NamedRerouteDeclaration(None=FunctionInput() {inputName=Normal}) {name=InNormal, nodeColor=<rgba 1 0 0.3984 1>}
  VectorParameter() {parameterName=Color 1, defaultValue=<rgba 0.9473 0 0.7715 1>}
 ]}
