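# Parent cloth-shading material for the TrafficHat backpack cosmetic, adding swirling liquid and bubbles to the 2023 anime character shader.
# MF_SwirlingLiquid sits between the anime Calculations and Post functions, with a masked lit/shaded color tint.
# When IsLiquid is on, PBR_Shading blends emissive toward time-of-day-scaled liquid color and also adjusts base color.
# MF_TrafficHat_Bubbles processes the final attributes using the anime FX map.
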Material /BRCosmetics/Accessories/FORT_Backpacks/Backpack_F_MED_TrafficHat/Materials/M_TrafficHat_Parent
graph {
 properties {
  bAutomaticallySetUsageInEditor=false
  bUseMaterialAttributes=true
  bUsedWithMeshDeformer=true
  bUsedWithMorphTargets=true
  bUsedWithSkeletalMesh=true
  bUsedWithStaticLighting=true
  pixelDepthOffsetMode=PDOM_Legacy
  shadingModel=MSM_Cloth
  usedShadingModels=Cloth
 }
 nodes [
  n1 = MaterialFunctionCall!(CustomUV) {materialFunction=/Game/Global/Materials/Character_Parent_2023/Anime_Parent_2023/MaterialFunctions/MFC_AnimeShader_2023_Pre}
  n2 = VectorParameter() {parameterName=ColorLit_Override, defaultValue=<rgba 1 1 1 1>}
  n3 = ChannelMaskParameter(None=@n1.FX) {parameterName=LiquidMask}
  LinearInterpolate(A=@n1.Color_Lit, B=Saturate(None=Multiply(A=@n1.Color_Lit, B=@n2.RGB)), Alpha=@n3.R)
  LinearInterpolate(A=@n1.Color_Shaded, B=Saturate(None=Multiply(A=@n1.Color_Shaded, B=@n2.RGB)), Alpha=@n3.R)
  n4 = MaterialFunctionCall!(CustomLightVector, OverrideNormal_WS, STT=@n1.STT, Normal_Map=@n1.STT, DFL=@n1.STT, FX=@n1.STT) {materialFunction=/Game/Global/Materials/Character_Parent_2023/Anime_Parent_2023/MaterialFunctions/MFC_AnimeShader_2023_Calculations}
  n5 = MaterialFunctionCall!(
   BaseColor=@n1.Color_Lit
   ShadedBaseColor=@n1.Color_Lit
   Normal=@n4."Normal_TS(PreBakedBlend)"
   UsePBR=ScalarParameter() {group=03_Anime_PBR_Override, parameterName=PBR_Shading, sliderMax=1, sortPriority=5}
   FXMask=@n1.Color_Lit
  ) {materialFunction=/BRCosmetics/Accessories/FORT_Backpacks/Backpack_F_MED_TrafficHat/Materials/MaterialFunctions/MF_SwirlingLiquid}
  n6 = MaterialFunctionCall!(
   Shading_Gradient=@n4.Shading_Gradient
   Threshold_Gradient=@n4.Shading_Gradient
   Specular_Gradient=@n4.Shading_Gradient
   Sheen_Mask=@n4.Shading_Gradient
   Thickness=@n4.Shading_Gradient
   DFL_Mask=@n4.Shading_Gradient
   FX=@n4.Shading_Gradient
   Metallic_Gradient=@n4.Shading_Gradient
   RimLight_Gradient=@n4.Shading_Gradient
   NormalBuffer=@n5.OutNormal
   --------
   CINE_LightSources=@n4.Shading_Gradient
   CINE_LightColor=@n4.Shading_Gradient
   ---------
   Color_Lit=@n5.OutNormal
   Color_Shaded=@n5.OutNormal
   SRM=@n1.SRM
   Emissive=@n1.SRM
   DFL_Color=@n1.SRM
   Specular_Color=@n1.SRM
   RimLight_Color=@n1.SRM
  ) {materialFunction=/Game/Global/Materials/Character_Parent_2023/Anime_Parent_2023/MaterialFunctions/MFC_AnimeShader_2023_Post}
  n7 = GetMaterialAttributes(None=@n6.Result) {attributeGetTypes [#1, #2]}
  n8 = Multiply(A=@n5.Color)
  n9 = Multiply(A=@n8) {constB=0.4}
  n10 = LinearInterpolate(
   A=@n7.Emissive Color
   B=Multiply(
    A=MaterialFunctionCall(Color_Morning=@n8, Color_Day=@n8, Color_Evening=@n9, Color_Night=@n9) {materialFunction=&Color_Time_of_Day}.Result
    B=ScalarParameter() {defaultValue=1, parameterName=LiquidPBREmissiveStrength}
   )
   Alpha=ScalarParameter() {group=03_Anime_PBR_Override, parameterName=PBR_Shading, sliderMax=1, sortPriority=5}
  )
  n11 = LinearInterpolate(
   A=@n7.Base Color
   B=Multiply(A=@n7.Base Color)
   Alpha=ScalarParameter() {group=03_Anime_PBR_Override, parameterName=PBR_Shading, sliderMax=1, sortPriority=5}
  ) {constB=0}
  MaterialFunctionCall!(
   InFXMap=@n1.FX
   In=StaticSwitchParameter(
    True=SetMaterialAttributes!(MaterialAttributes=@n7.MaterialAttributes, Emissive Color=@n10, Base Color=@n11) {attributeSetTypes [#1, #2]}
    False=@n6.Result
   ) {parameterName=IsLiquid}
  ) {materialFunction=/BRCosmetics/Accessories/FORT_Backpacks/Backpack_F_MED_TrafficHat/Materials/MaterialFunctions/MF_TrafficHat_Bubbles}
 ]
}
